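# Editor utility widget that previews a car or other chosen object on a DelMar track spline and simulates driving along it.
# Each tick traces from the viewport camera, writes the point to the DriverPos0 material parameter, and moves a transient preview actor there, rotated to the nearest spline transform.
# Drive mode advances along the spline's forward vector, re-snaps to it, and places the editor camera behind.
# Refresh destroys the preview actor so it respawns.
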EditorUtilityWidgetBlueprint /DelMarTools/EditorUtilities/TrackCarTester/EUW_TrackCarTester
graphs [
 EventGraph {nodes [
   "On Clicked (btnRefresh)"!() -> (OutputDelegate, then, widget:User Widget Object Reference, event:Pointer Event Structure) {is=title K2Node_ComponentBoundEvent}
   n1 = Get currentPreviewActor() -> (currentPreviewActor:Actor Object Reference)
   > Is Valid(InputObject=@n1)
   > DestroyActor!(execute, self:Editor Actor Subsystem Object Reference=Get EditorActorSubsystem(), ActorToDestroy:Actor Object Reference=@n1) -> (then, ReturnValue:bool)
   "On Clicked (btnDrive)"!() -> (OutputDelegate, then, widget:User Widget Object Reference, event:Pointer Event Structure) {is=title K2Node_ComponentBoundEvent}
   > Set isDriving(isDriving:bool=NOT Boolean(A=Get isDriving() -> (isDriving:bool))) -> (Output_Get:bool)
   > n2 = Get Editor Camera Info!(execute) -> (then, CameraLocation:Vector, ForwardDirection:Vector, FOV:float, ReturnValue:bool) {is=title K2Node_CallFunction}
   > SnapToTrack!(execute, self, WorldLocation:"Vector (by ref)"=@n2) -> (then)
   n3 = Event Tick()
   > Set tickDeltaTime(tickDeltaTime:float=@n3.InDeltaTime) -> (Output_Get:float)
   > HandleTick()
   Event Destruct()
   > Is Valid(InputObject=Get currentPreviewActor())
   > Set Actor Hidden In Game!(execute, self:Actor Object Reference=Get currentPreviewActor(), bNewHidden:bool=true) -> (then) {is=title K2Node_CallFunction}
   > SetVectorParameterValue!(
    execute
    WorldContextObject
    Collection:Material Parameter Collection Object Reference
    ParameterName:Name=DriverPos0
    ParameterValue:"Linear Color Structure (by ref)"=MakeColor!(R:float=0, G:float=0, B:float=0, A:float=1) -> (ReturnValue:Linear Color Structure)
   ) -> (then) {is=title K2Node_CallMaterialParameterCollectionFunction}
   Event Construct()
   > SetObject(
    self=[
     Get DetailsView_0() -> (DetailsView_0:Details View Object Reference)
     Get DetailsViewDrive() -> (DetailsViewDrive:Details View Object Reference)
    ]
    NewObject=Self-Reference()
   )
  ]}
 HandleTick {nodes [
   HandleTick() {is=title K2Node_FunctionEntry}
   > n1 = Sequence!(execute) -> (then_0, then_1, then_2:Exec, then_3:Exec)
   > n2 = Branch(Condition=Get isDriving())
   n3 = Get btnDrive() -> (btnDrive:TA Generic 01 Button Object Reference)
   > setLabel!(execute, self:TA Base Editor Button Widget Object Reference=@n3, label:String=Stop Driving) -> (then)
   n4 = BreakTransform(InTransform=Get currentDriveTransform() -> (currentDriveTransform:Transform))
   n5 = GetForwardVector!(InRot:Rotator=@n4.Rotation) -> (ReturnValue:Vector)
   n6 = float / float(
    A=Get DriveSpeed() -> (DriveSpeed:float)
    B=float * float(A=Get tickDeltaTime() -> (tickDeltaTime:float), B=60) {is=title K2Node_PromotableOperator}
   )
   > SnapToTrack(WorldLocation=vector + vector(A=@n4, B=vector * vector(A=@n5, B=@n6) -> (ReturnValue=~)))
   setLabel(execute=@n2.else, self=@n3, label=Start Driving)
   n7 = Get Editor Camera Info(execute=@n1.then_1)
   > Set currentCameraLoc(currentCameraLoc:Vector=@n7) -> (Output_Get:Vector)
   n8 = vector + vector(A=@n7.ForwardDirection, B="0,0,-0.2")
   n9 = Get PreviewProjectionDistance() -> (PreviewProjectionDistance:float)
   > Set currentProjectionLoc(currentProjectionLoc:Vector=vector + vector(A=@n7, B=vector * vector(A=@n8, B:float=@n9) -> (ReturnValue=~))) -> (Output_Get:Vector)
   n10 = Line Trace By Profile!(
    execute=@n1.then_2
    WorldContextObject:Object Reference
    Start:Vector=Get currentCameraLoc() -> (currentCameraLoc:Vector)
    End:Vector=Get currentProjectionLoc() -> (currentProjectionLoc:Vector)
    ProfileName:Name=FortBuildingMeshPhysics
    bTraceComplex:bool=false
    ActorsToIgnore:Array of Actor Object References=Make Array("[0]":Actor Object Reference=Get currentPreviewActor()) -> (Array:Array of Actor Object References)
    DrawDebugType:EDrawDebugTrace Enum
    bIgnoreSelf:bool=true
    TraceColor:Linear Color Structure=(R=1,G=0,B=0,A=1)
    TraceHitColor:Linear Color Structure=(R=0,G=1,B=0,A=1)
    DrawTime:float=5
   ) -> (then, OutHit:Hit Result Structure, ReturnValue:bool)
   > n11 = Branch(Condition=@n10.ReturnValue)
   n12 = BreakHitResult!(Hit:"Hit Result Structure (by ref)"=@n10) -> (
    bBlockingHit:bool
    bInitialOverlap:bool
    Time:float
    Distance:float
    Location:Vector
    ImpactPoint:Vector
    Normal:Vector
    ImpactNormal:Vector
    PhysMat:Physical Material Object Reference
    HitActor:Actor Object Reference
    HitComponent:Primitive Component Object Reference
    HitBoneName:Name
    BoneName:Name
    HitItem:int
    ElementIndex:int
    FaceIndex:int
    TraceStart:Vector
    TraceEnd:Vector
   )
   > n13 = Cast To SplineMeshComponent!(execute, Object=@n12.HitComponent) -> (then, CastFailed, AsSpline Mesh Component:Spline Mesh Component Object Reference) {is=title K2Node_DynamicCast}
   > Set currentFinalLoc(currentFinalLoc:Vector=@n12.Location) -> (Output_Get:Vector)
   > Set currentImpactDir(currentImpactDir:Vector=@n12.ImpactNormal) -> (Output_Get:Vector)
   Set currentFinalLoc(execute=[@n13.CastFailed, @n11.else], currentFinalLoc=Get currentProjectionLoc())
   > Set currentImpactDir(currentImpactDir="0, 0, 0")
   SetVectorParameterValue(
    execute=@n1.then_3
    ParameterName=DriverPos0
    ParameterValue="To LinearColor (Vector)"!(InVec:Vector=Get currentFinalLoc() -> (currentFinalLoc:Vector)) -> (ReturnValue:Linear Color Structure)
   )
   > SetMeshFriendLoc!(execute, self, location:Vector=Get currentFinalLoc()) -> (then)
  ]}
 SetMeshFriendLoc {nodes [
   n1 = SetMeshFriendLoc!() -> (then, location:Vector=~) {is=title K2Node_FunctionEntry}
   > Set l_loc(l_loc:Vector=@n1) -> (Output_Get:Vector)
   > getMeshFriend!(execute, self) -> (then, actor:Actor Object Reference)
   > n2 = Sequence!(execute) -> (then_0, then_1, then_2)
   > Set Actor Location!(execute, self:Actor Object Reference=Get currentPreviewActor(), NewLocation:Vector=@n1, bSweep:bool=false, bTeleport:bool=true) -> (then, SweepHitResult:Hit Result Structure, ReturnValue:bool) {is=title K2Node_CallFunction}
   > Set Actor Hidden In Game(self=Get currentPreviewActor(), bNewHidden=false)
   n3 = getTrack!(execute=@n2.then_1, self) -> (then, actor:Actor Object Reference)
   > n4 = Bad cast node!(execute, Object=@n3) -> (then, CastFailed, AsDel Mar Track:Object Reference) {is=title K2Node_DynamicCast}
   > Set l_transform(l_transform:Transform=FindTransformClosestToWorldLocation!(
     self:Spline Component Object Reference=Get Spline!(self:Object Reference=@n4) -> (Spline:Object Reference)
     WorldLocation:"Vector (by ref)"=Get l_loc() -> (l_loc:Vector)
     CoordinateSpace:ESplineCoordinateSpace Enum=World
     bUseScale:bool=false
    ) -> (ReturnValue:Transform)) -> (Output_Get:Transform)
   n5 = Get!() -> (l_transform_Location:Vector, l_transform_Rotation:Rotator, l_transform_Scale:Vector)
   > Set l_rot(l_rot:Rotator=@n5.l_transform_Rotation) -> (Output_Get:Rotator)
   n6 = Get currentPreviewActor()
   Set Actor Rotation!(execute=@n2.then_2, self:Actor Object Reference=@n6, NewRotation:Rotator=Get l_rot() -> (l_rot:Rotator), bTeleportPhysics:bool=true) -> (then, ReturnValue:bool) {is=title K2Node_CallFunction}
   > n7 = Branch(Condition=Get isDriving())
   Add Actor Local Rotation!(
    execute=@n7.else
    self:Actor Object Reference=@n6
    DeltaRotation:Rotator=MakeRotator(Yaw=Get RotatePreview() -> (RotatePreview:double))
    bSweep:bool=false
    bTeleport:bool=true
   ) -> (then, SweepHitResult:Hit Result Structure)
   float + float(A=BreakRotator(InRot=@n5.l_transform_Rotation).Yaw, B=Get RotatePreview())
  ]}
 SnapToTrack {nodes [
   n1 = SnapToTrack!() -> (then, WorldLocation:"Vector (by ref)") {is=title K2Node_FunctionEntry}
   > n2 = Sequence!(execute) -> (then_0, then_1)
   > n3 = getTrack()
   > n4 = Bad cast node(Object=@n3)
   > n5 = Set l_splineTransform(l_splineTransform:Transform=FindTransformClosestToWorldLocation(self=Get Spline(self=@n4), WorldLocation=@n1, CoordinateSpace=World, bUseScale=false)) -> (Output_Get:Transform)
   > Set currentDriveTransform(currentDriveTransform:Transform=@n5) -> (Output_Get:Transform)
   n6 = BreakRotator(InRot=BreakTransform(InTransform=@n5).Rotation)
   n7 = float * float(A=Get DriveCameraAngle() -> (DriveCameraAngle:double), B=-1)
   n8 = Get RotatePreview()
   > n9 = Set l_rot(l_rot=MakeRotator(Pitch=float + float(A=@n7, B=@n6.Pitch), Yaw=float - float(A=@n6.Yaw, B=@n8)))
   n10 = Get l_splineTransform() -> (l_splineTransform:Transform)
   n11 = GetForwardVector(InRot=@n9)
   n12 = Get PreviewProjectionDistance()
   > Set l_targetLoc(l_targetLoc:Vector=vector + vector(A=vector * vector!(A=@n11, B=-1, C:float=@n12) -> (ReturnValue=~), B=BreakTransform(InTransform=@n10))) -> (Output_Get:Vector)
   SetLevelViewportCameraInfo!(
    execute=@n2.then_1
    self:Unreal Editor Subsystem Object Reference=Get UnrealEditorSubsystem()
    CameraLocation:Vector=Get l_targetLoc() -> (l_targetLoc:Vector)
    CameraRotation:Rotator=Get l_rot()
   ) -> (then)
  ]}
 getMeshFriend {nodes [
   getMeshFriend() {is=title K2Node_FunctionEntry}
   > n1 = Is Valid(InputObject=Get currentPreviewActor())
   n2 = GetAllLevelActors(execute=@n1.Is Not Valid, self=Get EditorActorSubsystem())
   > n3 = Filter by Actor Tag!(
    execute
    TargetArray:Array of Actor Object References=@n2
    Tag:Name=Get MySpawnedActorTag() -> (MySpawnedActorTag:Name=None)
    FilterType:EEditorScriptingFilterType Enum
   ) -> (then, ReturnValue:Array of Actor Object References)
   > n4 = For Each Loop(Array:Array of Actor Object References=@n3) -> (Array Element:Actor Object Reference)
   n5 = SpawnActorFromObject!(
    execute=@n4.Completed
    self:Editor Actor Subsystem Object Reference=Get EditorActorSubsystem()
    ObjectToUse:Object Reference=Get PreviewObject() -> (PreviewObject:Object Reference)
    Location:Vector="0, 0, 0"
    Rotation:Rotator
    bTransient:bool=true
   ) -> (then, ReturnValue:Actor Object Reference)
   > n6 = Set currentPreviewActor(currentPreviewActor:Actor Object Reference=@n5) -> (Output_Get:Actor Object Reference)
   > Return Node!(execute, actor:Actor Object Reference=@n6)
   n7 = Set currentPreviewActor(execute=@n4.LoopBody, currentPreviewActor=@n4)
   > Return Node(actor=@n7)
   Return Node(execute=@n1.Is Valid, actor=Get currentPreviewActor())
  ]}
 getTrack {nodes [
   getTrack() {is=title K2Node_FunctionEntry}
   > n1 = Is Valid(InputObject=Get currentTrack() -> (currentTrack:Actor Object Reference))
   n2 = GetAllLevelActors(execute=@n1.Is Not Valid, self=Get EditorActorSubsystem())
   > n3 = Filter by Class!(execute, TargetArray:Array of Object References=@n2, ObjectClass:Object Class Reference, FilterType:EEditorScriptingFilterType Enum) -> (then, ReturnValue:Array of Object References)
   > Branch(Condition=Is Valid Index!(TargetArray:Array of Object References=@n3, IndexToTest:int) -> (ReturnValue:bool))
   > n4 = Set currentTrack(currentTrack:Actor Object Reference) -> (Output_Get:Actor Object Reference)
   > Return Node(actor=@n4)
   Return Node(execute=@n1.Is Valid, actor=Get currentTrack())
   "Get (a copy)"(Array:Array of Object References=@n3) -> (Output:Object Reference)
  ]}
]
variables [
 currentCameraLoc {type=Vector}
 currentDriveTransform {type=Transform, default {location=<0 0 0>, rotation=<rot 0 0 0>, scale=<1 1 1>}}
 currentFinalLoc {type=Vector}
 currentImpactDir {type=Vector}
 currentPreviewActor {type=Actor, default=null}
 currentProjectionLoc {type=Vector}
 currentTrack {type=Actor, default=null}
 DriveCameraAngle {type=number, default=20}
 DriveSpeed {type=number, default=150}
 isDriving {type=boolean}
 MySpawnedActorTag {type=string, default=OOBtubepreviewactor}
 PreviewObject {type=Object, default=null}
 PreviewProjectionDistance {type=number, default=1500}
 PreviewRotationOffset {type=Rotator}
 RotatePreview {type=number}
 tickDeltaTime {type=number}
]
widgets {
 info {rootWidgetClass=&VerticalBox, widgetCount=5}
 widgets [
  VerticalBox_0 {widgetClassPath=&VerticalBox, widget=VerticalBox_0}
  btnRefresh {
   widgetClassPath=&TA_Generic01_Button
   parent=VerticalBox_0
   widget=btnRefresh
   properties {label=Refresh Preview}
   slot {class=VerticalBoxSlot}
  }
  DetailsView_0 {
   widgetClassPath=&DetailsView
   parent=VerticalBox_0
   widget=DetailsView_0
   properties {categoriesToShow [Preview]}
   slot {class=VerticalBoxSlot}
  }
  btnDrive {
   widgetClassPath=&TA_Generic01_Button
   parent=VerticalBox_0
   widget=btnDrive
   properties {label=Drive}
   slot {class=VerticalBoxSlot, properties {padding=<ltrb 0 10 0 0>}}
  }
  DetailsViewDrive {
   widgetClassPath=&DetailsView
   parent=VerticalBox_0
   widget=DetailsViewDrive
   properties {categoriesToShow [Driving]}
   slot {class=VerticalBoxSlot}
  }
 ]
}
